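# Physics asset for the sixth-birthday cake back bling, providing secondary jiggle on the candles, flames, number six and cake knife.
# Capsule bodies form chains from the pack through master_cake to base anchors for five candles, the six and the knife, joined by free-rotation constraints.
# Above each anchor, tightly limited joints let candles and the six wobble with looser flames on top.
# The knife carries the cake slice and a two-segment handle with near-rigid limits.
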PhysicsAsset /BRCosmetics/Accessories/FORT_Backpacks/Backpack_M_MED_FNBirthdaySix/Meshes/M_MED_FNBirthdaySix_Pack_Physics
bodies {
 base_cake_knife [{shapeType=Capsule, length=0.505, radius=0.505, shapeName=base_cake_knife_capsule, rotation=<rot 0 -90 0>}]
 base_candle_01 [{shapeType=Capsule, length=0.505, radius=0.505, shapeName=base_candle_01_capsule, rotation=<rot 0 -90 0>}]
 base_candle_02 [{shapeType=Capsule, length=0.505, radius=0.505, shapeName=base_candle_02_capsule, rotation=<rot 0 -90 0>}]
 base_candle_03 [{shapeType=Capsule, length=0.505, radius=0.505, shapeName=base_candle_03_capsule, rotation=<rot 0 -90 0>}]
 base_candle_04 [{shapeType=Capsule, length=0.505, radius=0.505, shapeName=base_candle_04_capsule, rotation=<rot 0 -90 0>}]
 base_candle_05 [{shapeType=Capsule, length=0.505, radius=0.505, shapeName=base_candle_05_capsule, rotation=<rot 0 -90 0>}]
 base_six [{shapeType=Capsule, length=0.505, radius=0.505, shapeName=base_six_capsule, rotation=<rot 0 -90 0>}]
 dyn_cake_knife [{shapeType=Capsule, length=2, radius=1, shapeName=dyn_cake_knife_capsule, center=<2 0 0>, rotation=<rot 90 0 0>}]
 dyn_cake_slice [{shapeType=Capsule, length=2, radius=1, shapeName=dyn_cake_slice_capsule, center=<2 0 0>, rotation=<rot 90 0 0>}]
 dyn_candle_01 [{shapeType=Capsule, length=3.1124, radius=1.1762, shapeName=dyn_candle_01_capsule, center=<3.0408 0 0>, rotation=<rot 90 0 0>}]
 dyn_candle_02 [{shapeType=Capsule, length=3.2491, radius=1.2236, shapeName=dyn_candle_02_capsule, center=<3.3152 0 0>, rotation=<rot 90 0 0>}]
 dyn_candle_03 [{shapeType=Capsule, length=2.6528, radius=1.2131, shapeName=dyn_candle_03_capsule, center=<3.0154 0 0>, rotation=<rot 90 0 0>}]
 dyn_candle_04 [{shapeType=Capsule, length=3.2962, radius=1.1789, shapeName=dyn_candle_04_capsule, center=<3.2247 0 0>, rotation=<rot 90 0 0>}]
 dyn_candle_05 [{shapeType=Capsule, length=3.8004, radius=1.221, shapeName=dyn_candle_05_capsule, center=<3.6136 0 0>, rotation=<rot 90 0 0>}]
 dyn_flame_01 [{shapeType=Capsule, length=0.5122, radius=0.5, shapeName=dyn_flame_01_capsule, center=<0.5517 0 0>, rotation=<rot 90 0 0>}]
 dyn_flame_02 [{shapeType=Capsule, length=0.5221, radius=0.5, shapeName=dyn_flame_02_capsule, center=<0.5829 0 0>, rotation=<rot 90 0 0>}]
 dyn_flame_03 [{shapeType=Capsule, length=0.5198, radius=0.5, shapeName=dyn_flame_03_capsule, center=<0.6318 0 0>, rotation=<rot 90 0 0>}]
 dyn_flame_04 [{shapeType=Capsule, length=0.4958, radius=0.5089, shapeName=dyn_flame_04_capsule, center=<0.5809 -0.1289 -0.217>, rotation=<rot 90 0 0>}]
 dyn_flame_05 [{shapeType=Capsule, length=0.6555, radius=0.5, shapeName=dyn_flame_05_capsule, center=<0.6223 -0.1046 0.1076>, rotation=<rot 90 0 0>}]
 dyn_flame_06 [{shapeType=Capsule, length=1.3892, radius=0.7405, shapeName=dyn_flame_06_capsule, center=<1.2587 0 0>, rotation=<rot 90 0 0>}]
 dyn_handle_00 [{shapeType=Capsule, length=2.0792, radius=0.9208, shapeName=dyn_handle_00_capsule, center=<2 0 0>, rotation=<rot 90 0 0>}]
 dyn_handle_01 [{shapeType=Capsule, length=3, radius=1, shapeName=dyn_handle_01_capsule, center=<2.6307 0 0>, rotation=<rot 90 0 0>}]
 dyn_six [{shapeType=Capsule, length=1.5774, radius=3.1017, shapeName=dyn_six_capsule, center=<3.7573 0 0>, rotation=<rot 90 0 0>}]
 master_cake [{shapeType=Capsule, length=0.505, radius=0.505, shapeName=master_cake_capsule, rotation=<rot 0 -90 0>}]
 pack [{shapeType=Capsule, length=1.4944, radius=2.6843, shapeName=pack_capsule, center=<0.1296 -3.9955 0>, rotation=<rot 90 0 0>}]
}
constraints [
 {
  bone1Name=master_cake
  bone2Name=pack
  swing1LimitDegrees=45
  swing1Motion=Free
  swing2LimitDegrees=45
  swing2Motion=Free
  twistLimitDegrees=45
  twistMotion=Free
 }
 {
  bone1Name=base_candle_01
  bone2Name=master_cake
  swing1LimitDegrees=45
  swing1Motion=Free
  swing2LimitDegrees=45
  swing2Motion=Free
  twistLimitDegrees=45
  twistMotion=Free
 }
 {bone1Name=dyn_candle_01, bone2Name=base_candle_01, swing1LimitDegrees=3, swing2LimitDegrees=3, twistLimitDegrees=1}
 {bone1Name=dyn_flame_01, bone2Name=dyn_candle_01, swing1LimitDegrees=15, swing2LimitDegrees=10, twistLimitDegrees=1}
 {
  bone1Name=base_candle_02
  bone2Name=master_cake
  swing1LimitDegrees=45
  swing1Motion=Free
  swing2LimitDegrees=45
  swing2Motion=Free
  twistLimitDegrees=45
  twistMotion=Free
 }
 {bone1Name=dyn_candle_02, bone2Name=base_candle_02, swing1LimitDegrees=3, swing2LimitDegrees=3, twistLimitDegrees=1}
 {bone1Name=dyn_flame_02, bone2Name=dyn_candle_02, swing1LimitDegrees=15, swing2LimitDegrees=10, twistLimitDegrees=1}
 {
  bone1Name=base_candle_03
  bone2Name=master_cake
  swing1LimitDegrees=45
  swing1Motion=Free
  swing2LimitDegrees=45
  swing2Motion=Free
  twistLimitDegrees=45
  twistMotion=Free
 }
 {bone1Name=dyn_candle_03, bone2Name=base_candle_03, swing1LimitDegrees=3, swing2LimitDegrees=3, twistLimitDegrees=1}
 {bone1Name=dyn_flame_03, bone2Name=dyn_candle_03, swing1LimitDegrees=15, swing2LimitDegrees=10, twistLimitDegrees=1}
 {
  bone1Name=base_candle_04
  bone2Name=master_cake
  swing1LimitDegrees=45
  swing1Motion=Free
  swing2LimitDegrees=45
  swing2Motion=Free
  twistLimitDegrees=45
  twistMotion=Free
 }
 {bone1Name=dyn_candle_04, bone2Name=base_candle_04, swing1LimitDegrees=3, swing2LimitDegrees=3, twistLimitDegrees=1}
 {bone1Name=dyn_flame_04, bone2Name=dyn_candle_04, swing1LimitDegrees=10, swing2LimitDegrees=15, twistLimitDegrees=1}
 {
  bone1Name=base_candle_05
  bone2Name=master_cake
  swing1LimitDegrees=45
  swing1Motion=Free
  swing2LimitDegrees=45
  swing2Motion=Free
  twistLimitDegrees=45
  twistMotion=Free
 }
 {bone1Name=dyn_candle_05, bone2Name=base_candle_05, swing1LimitDegrees=3, swing2LimitDegrees=3, twistLimitDegrees=1}
 {bone1Name=dyn_flame_05, bone2Name=dyn_candle_05, swing1LimitDegrees=15, swing2LimitDegrees=10, twistLimitDegrees=1}
 {
  bone1Name=base_six
  bone2Name=master_cake
  swing1LimitDegrees=45
  swing1Motion=Free
  swing2LimitDegrees=45
  swing2Motion=Free
  twistLimitDegrees=45
  twistMotion=Free
 }
 {bone1Name=dyn_six, bone2Name=base_six, swing1LimitDegrees=4.5, swing2LimitDegrees=5, twistLimitDegrees=1}
 {bone1Name=dyn_flame_06, bone2Name=dyn_six, swing1LimitDegrees=10, swing2LimitDegrees=10, twistLimitDegrees=1}
 {bone1Name=dyn_handle_01, bone2Name=dyn_handle_00, swing1LimitDegrees=0.5, swing2LimitDegrees=0, swing2Motion=Locked, twistLimitDegrees=1}
 {bone1Name=dyn_handle_00, bone2Name=dyn_cake_knife, swing1LimitDegrees=1, swing2LimitDegrees=45, swing2Motion=Locked, twistLimitDegrees=1}
 {
  bone1Name=base_cake_knife
  bone2Name=master_cake
  swing1LimitDegrees=45
  swing1Motion=Free
  swing2LimitDegrees=45
  swing2Motion=Free
  twistLimitDegrees=45
  twistMotion=Free
 }
 {bone1Name=dyn_cake_knife, bone2Name=base_cake_knife, swing1LimitDegrees=2.5, swing2LimitDegrees=0, twistLimitDegrees=0}
 {bone1Name=dyn_cake_slice, bone2Name=dyn_cake_knife, swing1LimitDegrees=1, swing2LimitDegrees=0, twistLimitDegrees=0}
]
